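; GA_Quest_LightningSpears_Hits is a Battle Pass S24 quest gameplay ability that tracks lightning spear hits.
; On event activation it calls the parent implementation, prints debug messages including the current hit count to screen and log, then increments Count and ends the ability.
; The visible logic only counts and prints; any quest progress reporting would come from the parent class.
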(event Ability|EventActivateAbilityFromEvent
  (params (EventData :type Gameplay%20Event%20Data%20Structure%20%28by%20ref%29))
  (bind _count (Variables|Default|GetCount))
  (|CallParentFunction EventData)
  (Development|PrintString "SPLODEEEE!!!!!!" true true "(R=1.000000,G=0.062383,B=0.000000,A=1.000000)" 8.0)
  (Development|PrintString (Utilities|String|ToString(Integer) _count) true true "(R=1.000000,G=0.763262,B=0.000000,A=1.000000)" 8.0)
  (Variables|Default|SetCount (+ _count 1))
  (Ability|EndAbility))
